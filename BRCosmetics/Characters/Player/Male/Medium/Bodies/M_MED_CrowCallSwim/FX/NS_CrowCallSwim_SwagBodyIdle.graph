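# Niagara idle body effect for the CrowCallSwim character cosmetic, adding glows, lights and lightning attached to skeletal mesh bones.
# GlowsAndLights bursts nine partitioned particles with per-index bone, offset, size and color arrays, while HolsterGlow and ThighGlow add sprite glows.
# Lightning spawns randomly flipped meshes in chest, hand, back and eye groups.
# ReactivityAlpha drives flicker and lightning spawning; back, eye, head-hidden and ADS flags gate individual effects.
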NiagaraSystem /BRCosmetics/Characters/Player/Male/Medium/Bodies/M_MED_CrowCallSwim/FX/NS_CrowCallSwim_SwagBodyIdle
dependencies {
 usedDataInterfaces [
  &NiagaraDataInterfaceArrayFloat
  &NiagaraDataInterfaceArrayInt32
  &NiagaraDataInterfaceArrayFloat3
  &NiagaraDataInterfaceMeshRendererInfo
  &NiagaraDataInterfaceCurve
  &NiagaraDataInterfaceSkeletalMesh
 ]
 usedDynamicInputs [&SimulationPosition, &FNTimeOfDayScalar]
 usedModules [
  &self:SystemSpawnScript.SystemScriptSource.SystemScriptGraph.NiagaraNodeAssignment_4.SetVariables_#1
  &SystemState
  &self:SystemSpawnScript.SystemScriptSource.SystemScriptGraph.NiagaraNodeAssignment_5.SetVariables_#2
  &Timeline
  &self:SystemSpawnScript.SystemScriptSource.SystemScriptGraph.NiagaraNodeAssignment_6.SetVariables_#3
  &self:SystemSpawnScript.SystemScriptSource.SystemScriptGraph.NiagaraNodeAssignment_0.SetVariables_#4
  &self:SystemSpawnScript.SystemScriptSource.SystemScriptGraph.NiagaraNodeAssignment_8.SetVariables_#5
 ]
}
emitters {
 GlowsAndLights {
  emitterUpdateScript [
   &EmitterState(
    Inactive Response=::NewEnumerator1
    Life Cycle Mode=::NewEnumerator1
    Loop Behavior=::NewEnumerator1
    Loop Duration Mode=::NewEnumerator1
    UseLoopDelay=false
   )
   &SpawnBurst_Instantaneous(Spawn Count=9, Use Spawn Probability=false, UseLoopCountLimit=false)
  ]
  particleSpawnScript [
   &PartitionParticles(Number Of Partitions=9, Particles Per Partition=1, Partition Mode:ENiagara_PartitionMode=::NewEnumerator1)
   &InitializeParticle(Color Mode=::NewEnumerator0, UsePositionOffset=false, Write Material Random=true)
   /SetVariables(
    Particles.GlowsVisibilityTag:NiagaraInt32=&MergeHelper_IntSwitch(
     0=0
     1=0
     2=0
     3=0
     4=0
     5=&MakeCustomIntFromBool(Bool=$System.bShowBackFX, True INT=0, False INT=1)
     6=&MakeCustomIntFromBool(Bool=$System.bShowBackFX, True INT=0, False INT=1)
     7=&MakeCustomIntFromBool(Bool=$System.bShowEyeFX, True INT=0, False INT=1)
     8=&MakeCustomIntFromBool(Bool=$System.bShowEyeFX, True INT=0, False INT=1)
     Index=$Particles.PartitionParticles.Partition
     Num=::NewEnumerator7
    )
    Particles.VisibilityTag=&MakeIntFromFloat(Float=&MakeFloatFromInt(Int=&MergeHelper_IntSwitch(
       0=0
       1=0
       2=1
       3=1
       4=1
       5=&MakeCustomIntFromBool(Bool=$System.bShowBackFX, True INT=1, False INT=1)
       6=&MakeCustomIntFromBool(Bool=$System.bShowBackFX, True INT=1, False INT=1)
       7=&MakeCustomIntFromBool(Bool=$System.bShowEyeFX, True INT=1, False INT=1)
       8=&MakeCustomIntFromBool(Bool=$System.bShowEyeFX, True INT=1, False INT=1)
       Index=$Particles.PartitionParticles.Partition
       Num=::NewEnumerator7
      )))
   ) {isSetParameters=true}
  ]
  particleUpdateScript [
   &ParticleState(Kill Particles When Lifetime Has Elapsed=false, Let Infinitely Lived Particles Die When Emitter Deactivates=true)
   /SetVariables(Particles.MainColor:LinearColor=&MultiplyLinearColor(
     Linear Color=$System.MainColor
     Scale Factor=&MakeVector4_FromFloat(Float=&Waveform(
       Waveform Count=::NewEnumerator1
       Global Input Position=&ReturnNormalizedExecIndex(Normalized Index Scale=9)
       "Waveform [1]"=::NewEnumerator12
       "[1] Frequency"=2.65
       "[1] Amplitude Min/Max"=<0.7 1>
       "Waveform [2] Blend"=::NewEnumerator5
       "[2] Interpolation"=$System.ReactivityAlpha
       "[2] Invert Interpolation"=false
       "Waveform [2]"=::NewEnumerator12
       "[2] Input Position"=0
       "[2] Phase"=$Emitter.Age
       "[2] Phase Scale"=1
       "[2] Compound Frequency"=2
       "[2] Frequency"=16
       "[2] Amplitude Min/Max"=<0.75 1>
       "[2] Scale"=1
       "[2] Bias"=0
       "[2] Exponent"=1
       "[2] Pulse Duration"=0.1
       "[2] Clip Triangle Min/Max"=<0 0>
       "[2] Random Seed"=0
      ))
    )) {isSetParameters=true}
   /SetVariables(Particles.MainColor:LinearColor=&MakeCustomLinearColorFromBool(
     Bool=&Boolean_AndOperation(
      Bool1=&Boolean_OrOperation(Bool1=$System.bIsADS, Bool2=$System.bIsADSOther)
      Bool2=&SetBoolByIntComparison(A=$Particles.PartitionParticles.Partition, B=7, Comparison Type=::NewEnumerator4)
     )
     True Color=<rgba 0 0 0 0>
     False Color=$StackContext.MainColor
    )) {isSetParameters=true}
   /SetVariables(
    Particles.SoftGlowCameraOffset:NiagaraFloat=&SelectFloatFromArray(
     Direct Array Index=$Particles.PartitionParticles.Partition
     Float Selection Array=DataInterface() {propertyValues {FloatData [11,11,5,2,2,2,2,2,2]}}
    )
    Particles.SoftGlowColor:LinearColor=&MultiplyLinearColor(
     Linear Color=$StackContext.MainColor
     Scale Factor=&MakeFloat4FromVectorAndFloat(
      "Vector (RGB)"=&VectorFromFloat(Value=&SelectFloatFromArray(
        Direct Array Index=$Particles.PartitionParticles.Partition
        Float Selection Array=DataInterface() {propertyValues {FloatData [3,3,1,5,5,0.3,0.3,3,3]}}
       ))
      "Float (Alpha)"=0.5
     )
    )
    Particles.SoftGlowSize:Vector2f=&Multiply_Vector2D_ByFloat(
     Float=&MakeCustomFloatFromBool(
      Bool=&SetBoolByIntComparison(A=$Particles.PartitionParticles.Partition, B=7, Comparison Type=::NewEnumerator2)
      True Float=&MakeCustomFloatFromBool(Bool=$User.bisHeadHidden, True Float=0, False Float=1)
      False Float=1
     )
     Vector2D=&Vector2DFromFloat(Value=&SelectFloatFromArray(
       Direct Array Index=$Particles.PartitionParticles.Partition
       Float Selection Array=DataInterface() {propertyValues {FloatData [22,22,16,4,4,33,33,9,9]}}
      ))
    )
   ) {isSetParameters=true}
   /SetVariables(
    Particles.FakeLightColor:LinearColor=&MultiplyLinearColor(
     Linear Color=$StackContext.MainColor
     Scale Factor=&MakeFloat4FromVectorAndFloat(
      "Vector (RGB)"=&VectorFromFloat(Value=&SelectFloatFromArray(
        Direct Array Index=$Particles.PartitionParticles.Partition
        Float Selection Array=DataInterface() {propertyValues {FloatData [3,3,3,8,8,3,3,8,8]}}
       ))
      "Float (Alpha)"=0.25
     )
    )
    Particles.FakeLightSize:Vector2f=&Multiply_Vector2D_ByFloat(
     Float=&MakeCustomFloatFromBool(
      Bool=&SetBoolByIntComparison(A=$Particles.PartitionParticles.Partition, B=7, Comparison Type=::NewEnumerator2)
      True Float=&MakeCustomFloatFromBool(Bool=$User.bisHeadHidden, True Float=0, False Float=1)
      False Float=1
     )
     Vector2D=&Vector2DFromFloat(Value=&SelectFloatFromArray(
       Direct Array Index=$Particles.PartitionParticles.Partition
       Float Selection Array=DataInterface() {propertyValues {FloatData [33,33,22,8,8,16,16,11,11]}}
      ))
    )
   ) {isSetParameters=true}
   /SetVariables(
    Particles.CapsuleGlowColor:LinearColor=&MultiplyLinearColor(
     Linear Color=$StackContext.MainColor
     Scale Factor=&MakeFloat4FromVectorAndFloat(
      "Vector (RGB)"=&VectorFromFloat(Value=&SelectFloatFromArray(
        Direct Array Index=$Particles.PartitionParticles.Partition
        Float Selection Array=DataInterface() {propertyValues {FloatData [2,2,2,3,3,3,3,3,3]}}
       ))
      "Float (Alpha)"=1
     )
    )
    Particles.CapsuleSpriteSize:Vector2f=&Multiply_Vector2D_ByFloat(
     Float=&MakeCustomFloatFromBool(
      Bool=&SetBoolByIntComparison(A=$Particles.PartitionParticles.Partition, B=7, Comparison Type=::NewEnumerator2)
      True Float=&MakeCustomFloatFromBool(Bool=$User.bisHeadHidden, True Float=0, False Float=1)
      False Float=1
     )
     Vector2D=&Vector2DFromFloat(Value=&SelectFloatFromArray(
       Direct Array Index=$Particles.PartitionParticles.Partition
       Float Selection Array=DataInterface() {propertyValues {FloatData [11,11,8,4,4,6,6,3,3]}}
      ))
    )
   ) {isSetParameters=true}
   /SetVariables(
    Particles.LightColor=&MultiplyLinearColor(
     Linear Color=$StackContext.MainColor
     Scale Factor=&MakeFloat4FromVectorAndFloat(
      "Vector (RGB)"=&VectorFromFloat(Value=&SelectFloatFromArray(
        Direct Array Index=$Particles.PartitionParticles.Partition
        Float Selection Array=DataInterface() {propertyValues {FloatData [1,1,0,0,0,1,1,1,1]}}
       ))
      "Float (Alpha)"=33
     )
    )
    Particles.LightRadius=&Multiply_Float(
     A=&SelectFloatFromArray(
      Direct Array Index=$Particles.PartitionParticles.Partition
      Float Selection Array=DataInterface() {propertyValues {FloatData [33,33,0,0,0,11,11,3,3]}}
     )
     B=&MakeCustomFloatFromBool(
      Bool=&SetBoolByIntComparison(A=$Particles.PartitionParticles.Partition, B=7, Comparison Type=::NewEnumerator2)
      True Float=&MakeCustomFloatFromBool(Bool=$User.bisHeadHidden, True Float=0, False Float=1)
      False Float=1
     )
    )
   ) {isSetParameters=true}
   &SkeletalMeshLocation(
    Bone / Socket Index=&SelectIntFromArray(
     Direct Array Index=$Particles.PartitionParticles.Partition
     Int Selection Array=DataInterface() {propertyValues {IntData [0,0,0,1,1,0,0,2,2]}}
    )
    Bone Sampling Mode=::NewEnumerator4
    MaskSpawnGroup=false
    Sampled Position Offset=&SelectVectorFromArray(
     Direct Array Index=$Particles.PartitionParticles.Partition
     Vector Selection Array=DataInterface() {propertyValues {FloatData [<6 12 9>, <6 12 -9>, <2.5 15 0>, <6 10 1.8>, <6 10 -1.8>, <15 -12 -2.5>, <15 -12 2.5>, <4.25 9.1 -2.6>, <4.25 9.1 2.6>]}}
    )
    Skeletal Mesh=$System.Skeletal Mesh
   )
   /SetVariables(Particles.CapsuleSpriteAlignment:Vector3f=&RotateVectorByQuaternion(
     Vector To Rotate=&SelectVectorFromArray(
      Direct Array Index=$Particles.PartitionParticles.Partition
      Vector Selection Array=DataInterface() {propertyValues {FloatData [<0.5 -0.375 1>, <-0.5 0.375 1>, <1 -0.1 0>, <0.5 0 1>, <0.5 0 -1>, <0.5 0.333 -0.5>, <0.5 0.333 0.5>, <0 0.25 1>, <0 -0.25 1>]}}
     )
     Quaternion=$Particles.SkeletalMeshLocation.SampledRotation
    )) {isSetParameters=true}
   /SetVariables(Particles.CapsuleSpriteDynamic:Vector4f=&MakeFloat4FromVectorAndFloat(
     "Vector (RGB)"=&NormalizeVector(Vector To Normalize=$Particles.CapsuleSpriteAlignment, Safe Normalize=true, Safe Fallback Threshold=0, Safe Fallback Vector=<0 0 0.001>)
     "Float (Alpha)"=&SelectFloatFromArray(
      Direct Array Index=$Particles.PartitionParticles.Partition
      Float Selection Array=DataInterface() {propertyValues {FloatData [22,22,15,7,7,13,13,5,5]}}
     )
    )) {isSetParameters=true}
  ]
  rendererClasses [&NiagaraSpriteRendererProperties, &NiagaraLightRendererProperties]
  renderers [
   {rendererClass=&NiagaraSpriteRendererProperties}
   {rendererClass=&NiagaraSpriteRendererProperties, rendererIndex=1}
   {rendererClass=&NiagaraSpriteRendererProperties, rendererIndex=2}
   {rendererClass=&NiagaraLightRendererProperties, rendererIndex=3}
  ]
 }
 HolsterGlow {
  emitterUpdateScript [
   &EmitterState(
    Inactive Response=::NewEnumerator1
    Life Cycle Mode=::NewEnumerator1
    Loop Behavior=::NewEnumerator1
    Loop Duration Mode=::NewEnumerator1
    UseLoopDelay=false
   )
   &SpawnBurst_Instantaneous(Spawn Count=2, Use Spawn Probability=false, UseLoopCountLimit=false)
  ]
  particleSpawnScript [
   &InitializeParticle(
    Color=$System.MainColor
    Lifetime=5
    Sprite Size Mode=::NewEnumerator3
    Uniform Sprite Size=22
    UsePositionOffset=false
    Write Material Random=true
   )
   &PartitionParticles()
  ]
  particleUpdateScript [
   &ParticleState(Kill Particles When Lifetime Has Elapsed=false)
   &SkeletalMeshLocation(
    Bone / Socket Index=&SelectIntFromArray(Direct Array Index=$Particles.PartitionParticles.Partition, Int Selection Array=DataInterface() {propertyValues {IntData [3,4]}})
    Bone Sampling Mode=::NewEnumerator4
    MaskSpawnGroup=false
    Sampled Position Offset=&SelectVectorFromArray(
     Direct Array Index=$Particles.PartitionParticles.Partition
     Vector Selection Array=DataInterface() {propertyValues {FloatData [<15 -2.5 9>, <-15 2.5 -9>]}}
    )
    Skeletal Mesh=$System.Skeletal Mesh
   )
   &ScaleColor(Scale RGB=&VectorFromFloat(Value=1), ScaleA=true, ScaleRGB=true)
   &CameraOffset(Camera Offset Amount=1)
   /SetVariables(Particles.DynamicMaterialParameter:Vector4f=&MakeFloat4FromVectorAndFloat(
     "Vector (RGB)"=&RotateVectorByQuaternion(Vector To Rotate=<1 0 -0.1>, Quaternion=$StackContext.SkeletalMeshLocation.SampledRotation)
     "Float (Alpha)"=44
    )) {isSetParameters=true}
  ]
  rendererClasses [&NiagaraSpriteRendererProperties]
  renderers [{rendererClass=&NiagaraSpriteRendererProperties}]
 }
 Lightning {
  emitterSpawnScript [/SetVariables(
    Emitter.LightningTilingBack=1
    Emitter.LightningTilingChest=1.4
    Emitter.LightningTilingEye=1
    Emitter.LightningTilingHand=1.3
    Emitter.LightningWidthBack=6
    Emitter.LightningWidthChest=13
    Emitter.LightningWidthEye=6
    Emitter.LightningWidthHand=6
   ) {isSetParameters=true}]
  emitterUpdateScript [
   &EmitterState(
    Inactive Response=::NewEnumerator1
    Life Cycle Mode=::NewEnumerator1
    Loop Behavior=::NewEnumerator1
    Loop Duration Mode=::NewEnumerator1
    UseLoopDelay=false
   )
   &SpawnRate(
    Spawn Probability=&Lerp_Float(A=0.333, B=0.8, Alpha=$System.ReactivityAlpha)
    SpawnRate=5
    Use Spawn Probability=true
   )
   &SpawnRate(
    Spawn Probability=&Multiply_Float(
     A=&MakeCustomFloatFromBool(Bool=$System.bShowEyeFX, True Float=1, False Float=0)
     B=&Lerp_Float(A=0.333, B=1, Alpha=$System.ReactivityAlpha)
    )
    SpawnGroup=3
    SpawnRate=4
    Use Spawn Probability=true
   ) {module=SpawnRate001}
   &SpawnRate(
    Spawn Probability=&Lerp_Float(A=0.333, B=0.666, Alpha=$System.ReactivityAlpha)
    SpawnGroup=1
    SpawnRate=3
    Use Spawn Probability=true
   ) {module=SpawnRate002}
   &SpawnRate(
    Spawn Probability=&Multiply_Float(
     A=&MakeCustomFloatFromBool(Bool=$System.bShowBackFX, True Float=1, False Float=0)
     B=&Lerp_Float(A=0.333, B=0.75, Alpha=$System.ReactivityAlpha)
    )
    SpawnGroup=2
    SpawnRate=6
    Use Spawn Probability=true
   ) {module=SpawnRate003}
  ]
  particleSpawnScript [
   /SetVariables(Particles.SpawnGroup=$Emitter.SpawnGroup) {isSetParameters=true}
   &InitializeParticle(
    Color Mode=::NewEnumerator0
    Lifetime=&Lerp_Float(A=0.444, B=0.666, Alpha=$System.ReactivityAlpha)
    UsePositionOffset=false
    Write Material Random=true
   )
   /SetVariables(Particles.bFlipMesh:NiagaraBool=&RandomBool(Randomness Mode=::NewEnumerator2, Random Seed=?)) {isSetParameters=true}
   /SetVariables(Particles.MeshIndex=&Add_Integer(
     A=&Multiply_Int(A=$Particles.SpawnGroup, B=3)
     B=&RandomRangeInt(Minimum=0, Maximum=2)
    )) {isSetParameters=true}
   /SetVariables(Particles.Scale=&MakeCustomVectorFromBool(Bool=$Particles.bFlipMesh, True Vector=<1 1 -1>, False Vector=<1 1 1>)) {isSetParameters=true}
  ]
  particleUpdateScript [
   /SetVariables(Particles.MainColor:LinearColor=&MultiplyLinearColor(
     Linear Color=&MultiplyLinearColor(
      Linear Color=$System.MainColor
      Scale Factor=&MakeFloat4FromVectorAndFloat("Vector (RGB)"=&VectorFromFloat(Value=33), "Float (Alpha)"=2)
     )
     Scale Factor=&MakeVector4_FromFloat(Float=&Waveform(
       Waveform Count=::NewEnumerator1
       Global Input Position=&RandomRangeFloat(Minimum=0, Maximum=9)
       "Waveform [1]"=::NewEnumerator12
       "[1] Frequency"=6
       "[1] Amplitude Min/Max"=<0.5 1>
       "Waveform [2] Blend"=::NewEnumerator5
       "[2] Interpolation"=$System.ReactivityAlpha
       "[2] Invert Interpolation"=false
       "Waveform [2]"=::NewEnumerator12
       "[2] Input Position"=0
       "[2] Phase"=$Emitter.Age
       "[2] Phase Scale"=1
       "[2] Compound Frequency"=2
       "[2] Frequency"=33
       "[2] Amplitude Min/Max"=<0.333 1>
       "[2] Scale"=1
       "[2] Bias"=0
       "[2] Exponent"=1
       "[2] Pulse Duration"=0.1
       "[2] Clip Triangle Min/Max"=<0 0>
       "[2] Random Seed"=0
      ))
    )) {isSetParameters=true}
   /SetVariables(Particles.MainColor:LinearColor=&MakeCustomLinearColorFromBool(
     Bool=&Boolean_AndOperation(
      Bool1=&Boolean_OrOperation(Bool1=$System.bIsADS, Bool2=$System.bIsADSOther)
      Bool2=&Boolean_AndOperation(
       Bool1=&SetBoolByIntComparison(A=$Particles.SpawnGroup, B=3, Comparison Type=::NewEnumerator4)
       Bool2=&Boolean_NotOperation(Bool=$Particles.bFlipMesh)
      )
     )
     True Color=<rgba 0 0 0 0>
     False Color=$StackContext.MainColor
    )) {isSetParameters=true}
   &DynamicMaterialParameters(
    Index 0 Param 1=&MergeHelper_FloatSwitch(
     0=$Emitter.LightningWidthChest
     1=$Emitter.LightningWidthHand
     2=$Emitter.LightningWidthBack
     3=$Emitter.LightningWidthEye
     Index=$Particles.SpawnGroup
     Num=::NewEnumerator2
    )
    Index 0 Param 2=&MergeHelper_FloatSwitch(
     0=$Emitter.LightningTilingChest
     1=$Emitter.LightningTilingHand
     2=$Emitter.LightningTilingBack
     3=$Emitter.LightningTilingEye
     Index=$Particles.SpawnGroup
     Num=::NewEnumerator2
    )
    Index 0 Param 3=&RemapRange(
     Input Value To Remap=&FloatFromCurve(FloatCurve=DataInterface() {propertyValues {Curve {keys [<at 0 1>,<at 0.1 0.375>,<at 0.5 0.05>,<at 1 0>]}}})
     Input Min=0
     Input Max=1
     Output Min=&Lerp_Float(A=0.2, B=0, Alpha=$System.ReactivityAlpha)
     Output Max=1
    )
    Index 1 Param 1=&MakeFloatFromVector(Vector=$Particles.Scale, Channel=::NewEnumerator0)
    Index 1 Param 2=&MakeFloatFromVector(Vector=$Particles.Scale, Channel=::NewEnumerator1)
    Index 1 Param 3=&MakeFloatFromVector(Vector=$Particles.Scale, Channel=::NewEnumerator2)
    Module.Index0Param0Write=true
    Module.Index0Param1Write=true
    Module.Index0Param2Write=true
    Module.Index0Param3Write=false
    Module.Index1Param0Write=true
    Module.Index1Param1Write=true
    Module.Index1Param2Write=true
    Module.Index1Param3Write=true
    Write Parameter Index 1=true
   )
   &ParticleState()
   &SkeletalMeshLocation(
    Bone / Socket Index=&SelectIntFromArray(
     Direct Array Index=$Particles.SpawnGroup
     Direct Array Mode=::NewEnumerator1
     Int Selection Array=DataInterface() {propertyValues {IntData [0,1,0,2]}}
    )
    Bone Sampling Mode=::NewEnumerator4
    MaskSpawnGroup=false
    Skeletal Mesh=$System.Skeletal Mesh
   )
   &ScaleColor(
    Color Value To Scale=$StackContext.MainColor
    Scale Alpha=&FloatFromCurve(FloatCurve=DataInterface() {propertyValues {Curve {keys [
         {arriveTangent=9.0725, interpMode=RCIM_Cubic, leaveTangent=9.0725, tangentMode=RCTM_User, time=0, value=0}
         {arriveTangent=-2.5249, interpMode=RCIM_Cubic, leaveTangent=-0.2678, tangentMode=RCTM_Break, time=0.1, value=1}
         {arriveTangent=-1.6791, interpMode=RCIM_Cubic, leaveTangent=-0.0405, tangentMode=RCTM_Break, time=1, value=0}
        ]}}})
    ScaleA=true
    ScaleRGB=true
   )
   /SetVariables(Particles.VisibilityTag=&MakeCustomIntFromBool(
     Bool=&SetBoolByIntComparison(A=&SpawnGroup(), B=3, Comparison Type=::NewEnumerator4)
     True INT=&MakeCustomIntFromBool(Bool=$User.bisHeadHidden, True INT=1, False INT=0)
     False INT=0
    )) {isSetParameters=true}
  ]
  rendererClasses [&NiagaraMeshRendererProperties]
  renderers [{rendererClass=&NiagaraMeshRendererProperties, rendererIndex=0}]
 }
 ThighGlow {
  emitterUpdateScript [
   &EmitterState(
    Inactive Response=::NewEnumerator1
    Life Cycle Mode=::NewEnumerator1
    Loop Behavior=::NewEnumerator1
    Loop Duration Mode=::NewEnumerator1
    UseLoopDelay=false
   )
   &SpawnBurst_Instantaneous(Use Spawn Probability=false, UseLoopCountLimit=false)
  ]
  particleSpawnScript [&InitializeParticle(
    Color=$System.MainColor
    Lifetime=5
    Sprite Size Mode=::NewEnumerator3
    Uniform Sprite Size=33
    UsePositionOffset=false
    Write Material Random=true
   )]
  particleUpdateScript [
   &ParticleState(Kill Particles When Lifetime Has Elapsed=false)
   &SkeletalMeshLocation(
    Bone / Socket Index=3
    Bone Sampling Mode=::NewEnumerator4
    MaskSpawnGroup=false
    Sampled Position Offset=<13 -10 -0.25>
    Skeletal Mesh=$System.Skeletal Mesh
   )
   &ScaleColor(ScaleA=true, ScaleRGB=true)
   &CameraOffset(Camera Offset Amount=11)
  ]
  rendererClasses [&NiagaraSpriteRendererProperties]
  renderers [{rendererClass=&NiagaraSpriteRendererProperties}]
 }
}
summary {
 systemName=NS_CrowCallSwim_SwagBodyIdle
 emitters [
  {
   emitterName=GlowsAndLights
   rendererClasses [&NiagaraSpriteRendererProperties, &NiagaraLightRendererProperties]
  }
  {emitterName=Lightning, rendererClasses [&NiagaraMeshRendererProperties]}
  {emitterName=ThighGlow, rendererClasses [&NiagaraSpriteRendererProperties]}
  {emitterName=HolsterGlow, rendererClasses [&NiagaraSpriteRendererProperties]}
 ]
 userVariables [
  User.bIsADS {type=NiagaraBool}
  User.bIsADSOther {type=NiagaraBool}
  User.bIsFrontEnd {type=NiagaraBool}
  User.bIsFrontEndPreview {type=NiagaraBool}
  User.bisHeadHidden {type=NiagaraBool}
  User.bShowBackFX {type=NiagaraBool, defaultValue=true}
  User.bShowEyeFX {type=NiagaraBool, defaultValue=true}
  User.DebugReactivity {type=NiagaraBool}
  User.KillReactivityCount {type=NiagaraInt32}
 ]
}
