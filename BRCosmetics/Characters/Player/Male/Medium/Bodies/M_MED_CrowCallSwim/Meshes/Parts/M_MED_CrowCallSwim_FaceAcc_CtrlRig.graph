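# Control Rig for the M_MED_CrowCallSwim face accessory mesh that adjusts the cap and hair bones according to head orientation.
# A spherical pose reader on the head bone outputs a weight for chained Modify Transforms nodes that offset hair_base, cap and hair_base_R; a second hair_base_R node holds an identity transform.
# A final cap offset scales that weight by the AssaultRifleSelected variable, so it applies only when that value is nonzero.
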ControlRigBlueprint /BRCosmetics/Characters/Player/Male/Medium/Bodies/M_MED_CrowCallSwim/Meshes/Parts/M_MED_CrowCallSwim_FaceAcc_CtrlRig
graphs [
 RigVMFunctionLibrary {links [], nodes []}
 RigVMModel {nodes [
   RigUnit_BeginExecution = &RigUnit_BeginExecution()
   n1 = &RigUnit_SphericalPoseReader(
    ActiveRegionScaleFactors=(PositiveWidth=1.5,NegativeWidth=0.5,PositiveHeight=3,NegativeHeight=3)
    ActiveRegionSize=0.1
    Debug=(bDrawDebug=false,bDraw2D=False,bDrawLocalAxes=False,DebugScale=25,DebugSegments=20,DebugThickness=0.25)
    DriverAxis=(X=0,Y=-1,Z=0)
    DriverItem=(Type=Bone,Name="head")
    FalloffRegionScaleFactors=(PositiveWidth=0.7,NegativeWidth=0.5,PositiveHeight=0.7,NegativeHeight=0.7)
    FalloffSize=0.3
    FlipHeightScaling=False
    FlipWidthScaling=False
    OptionalParentItem=(Type=Bone,Name="None")
    RotationOffset=(X=-90,Y=0,Z=-55)
    ExecutePin=@RigUnit_BeginExecution.ExecutePin
   )
   > &RigUnit_ModifyTransforms(
    ItemToModify=((Item=(Type=Bone,Name="hair_base"),Transform=(Rotation=(X=0,Y=0,Z=0.0872,W=0.9962),Translation=(X=2.5,Y=0,Z=0),Scale3D=(X=1,Y=1,Z=1))))
    WeightMinimum=0
    Weight=@n1.OutputParam
   )
   > &RigUnit_ModifyTransforms(
    ItemToModify=((Item=(Type=Bone,Name="cap"),Transform=(Rotation=(X=-0.2837,Y=0.0418,Z=0.0124,W=0.9579),Translation=(X=0,Y=0,Z=0.5),Scale3D=(X=1,Y=1,Z=1))))
    WeightMinimum=0
    Weight=@n1.OutputParam
   )
   > &RigUnit_ModifyTransforms(
    ItemToModify=((Item=(Type=Bone,Name="hair_base_R"),Transform=(Rotation=(X=0,Y=-0.1736,Z=0,W=0.9848),Translation=(X=0,Y=0,Z=0),Scale3D=(X=1,Y=1,Z=1))))
    WeightMinimum=0
    Weight=@n1.OutputParam
   )
   > &RigUnit_ModifyTransforms(
    ItemToModify=((Item=(Type=Bone,Name="hair_base_R"),Transform=(Rotation=(X=0,Y=0,Z=0,W=1),Translation=(X=0,Y=0,Z=0),Scale3D=(X=1,Y=1,Z=1))))
    WeightMinimum=0
    Weight=@n1.OutputParam
   )
   Multiply = &RigVMFunction_MathDoubleMul(B=RigVMVariableNode() {variable=AssaultRifleSelected}.Value, A=@n1.OutputParam)
   > &RigUnit_ModifyTransforms(
    ItemToModify=((Item=(Type=Bone,Name="cap"),Transform=(Rotation=(X=-0.0992,Y=0.0825,Z=0.0516,W=0.9903),Translation=(X=0,Y=0,Z=0.3),Scale3D=(X=1,Y=1,Z=1))))
    WeightMinimum=0
    Weight=@Multiply.Result
   )
  ]}
]
hierarchy [
 root {loc=<0 0 0>, rot=<0 0 0 1>}
 pelvis {parent=root, loc=<0 -0.7071 93.1191>, rot=<0.0127 0.707 -0.0127 -0.707>}
 spine_01 {parent=pelvis, loc=<6.7494 -0.6388 0>, rot=<0 0 0.034 -0.9994>}
 spine_02 {parent=spine_01, loc=<7.2518 -0.035 0>, rot=<0 0 0.0238 -0.9997>}
 spine_03 {parent=spine_02, loc=<6.0104 0 0>, rot=<0 0 0.0149 -0.9999>}
 spine_04 {parent=spine_03, loc=<8.1024 0 0>, rot=<0 0 -0.0132 -0.9999>}
 spine_05 {parent=spine_04, loc=<10.0146 0 0>, rot=<0 0 -0.0015 -1>}
 neck_01 {parent=spine_05, loc=<11.3138 0.1529 0>, rot=<0 0 -0.2003 -0.9797>}
 neck_02 {parent=neck_01, loc=<4.5627 0 0>, rot=<0 0 0.0504 -0.9987>}
 head {parent=neck_02, loc=<5.3914 0 0>, rot=<0 0 0.115 -0.9934>}
 hair_base {parent=head, loc=<-2.8926 -3.5926 0>, rot=<0 0 0 1>}
 hair_dyn {parent=hair_base, loc=<0 0 0>, rot=<0 0 0.9271 -0.3748>}
 hair_A_l {parent=hair_dyn, loc=<0.2531 -1.6062 -5.8579>, rot=<0.9061 -0.1376 -0.0546 -0.3963>}
 hair_B_l {parent=hair_dyn, loc=<1.884 0.1267 -5.4856>, rot=<0.8679 -0.2979 0.1333 -0.3745>}
 hair_C_l {parent=hair_dyn, loc=<2.8032 0.1533 -4.466>, rot=<0.8565 -0.0828 0.1377 -0.4905>}
 hair_D_l {parent=hair_dyn, loc=<2.8509 1.1553 -2.4545>, rot=<0.8137 -0.1427 0.0703 -0.5591>}
 hair_E_l {parent=hair_dyn, loc=<4.0838 0.7612 -0.6133>, rot=<0.6313 -0.0817 0.2552 -0.7278>}
 hair_E_r {parent=hair_dyn, loc=<3.9025 0.5444 1.2288>, rot=<0.6053 -0.0644 0.0828 -0.789>}
 hair_A_r1 {parent=hair_dyn, loc=<1.4594 -1.6629 6.0333>, rot=<0.3288 -0.2144 0.4695 -0.7909>}
 hair_base_R {parent=hair_dyn, loc=<1.1289 -0.4728 2.413>, rot=<0.4358 0.372 0.2155 -0.7907>}
 hair_D_r {parent=hair_base_R, loc=<2.0525 1.4732 -0.1935>, rot=<0.2421 -0.19 -0.0976 -0.9464>}
 hair_C_r {parent=hair_base_R, loc=<2.7231 0.2364 0.0891>, rot=<0.0989 -0.332 -0.1304 -0.929>}
 hair_B_r {parent=hair_base_R, loc=<3.0354 -0.7575 0.1996>, rot=<0.1248 -0.3231 -0.0565 -0.9364>}
 hair_A_r {parent=hair_base_R, loc=<3.038 -2.3223 0.0998>, rot=<0.1362 -0.5974 -0.015 -0.7901>}
 cap {parent=head, loc=<3.7378 -4.2745 0>, rot=<0.0029 -0.7071 -0.0029 -0.7071>}
]
variables [AssaultRifleSelected {type=double, default=0}]
